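# Parent material for the Range vehicle wheel, adding colored rim and tire glow to the shared MF_WheelParent_Simple attributes.
# Rim glow uses a mask texture; tire glow sums two panning geometric patterns masked by red and green channels, one hard-stepped with Floor.
# Vertex red blends rim to tire glow, each behind a static switch, tinted by CustomEmissiveColor.
# The result passes through vehicle damage and de-rez transition functions.
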Material /VehicleCosmetics/Wheels/Range/Materials/M_WheelParent_Range
graph {
 properties {
  bAutomaticallySetUsageInEditor=false
  bUseMaterialAttributes=true
  bUsedWithMeshDeformer=true
  bUsedWithSkeletalMesh=true
  pixelDepthOffsetMode=PDOM_Legacy
  previewMesh=/VehicleCosmetics/Wheels/Range/Meshes/SK_Wheel_Range
  usedShadingModels=Default Lit
 }
 nodes [
  n1 = GetMaterialAttributes(None=MaterialFunctionCall() {materialFunction=&MF_WheelParent_Simple}.MaterialAttributes) {attributeGetTypes [#1]}
  n2 = TextureSampleParameter2D() {
   group=SpecialEffect
   parameterName=RimEmissive
   samplerType=SAMPLERTYPE_LinearColor
   texture=/VehicleCosmetics/Wheels/Range/Textures/T_Range_Rim_M
  }
  n3 = Constant()
  n4 = StaticSwitchParameter(True=Multiply(A=@n2.R, B=ScalarParameter() {defaultValue=1, group=SpecialEffect, parameterName=RimEmmisuveGlow}), False=@n3) {defaultValue=true, group=SpecialEffect, parameterName=RimLights}
  n5 = Panner(Coordinate=TextureCoordinate() {uTiling=0.33}) {speedX=0.5}
  n6 = TextureSampleParameter2D() {parameterName=TireEmissive, samplerType=SAMPLERTYPE_LinearColor, texture=/VehicleCosmetics/Wheels/Range/Textures/T_Range_Tire_E}
  n7 = ScalarParameter() {defaultValue=1, group=SpecialEffect, parameterName=TireEmissiveGlow}
  n8 = Floor(None=Multiply(A=TextureSample(UVs=@n5) {texture=/VehicleCosmetics/Wheels/_SharedTextures/FX/Skin_GeometricBM_RGB}.R, B=Multiply(A=@n6.R, B=@n7)))
  n9 = TextureCoordinate() {uTiling=0.33}
  n10 = Multiply(
   A=TextureSample(UVs=Panner(Coordinate=@n9) {speedX=0.4}) {texture=&Skin_GeometricBM_RGB}.R
   B=Multiply(A=@n7, B=@n6.G)
  )
  n11 = StaticSwitchParameter(True=Clamp(None=Add(A=@n8, B=@n10)) {maxDefault=3}, False=@n3) {defaultValue=true, group=SpecialEffect, parameterName=TireLights}
  n12 = VertexColor()
  n13 = Constant() {r=3}
  n14 = VectorParameter() {group=SpecialEffect, parameterName=CustomEmissiveColor, defaultValue=<rgba 0.1 0.4 1 1>}
  n15 = Add(A=@n1.Emissive Color, B=Multiply(A=LinearInterpolate(A=@n4, B=@n11, Alpha=@n12.R), B=Multiply(A=@n13, B=@n14.RGB)))
  MaterialFunctionCall(In=MaterialFunctionCall(Material=SetMaterialAttributes(MaterialAttributes=@n1.MaterialAttributes, Emissive Color=@n15) {attributeSetTypes [#1]}) {materialFunction=&MF_BaseVehicle_Damage}.Result) {materialFunction=&MF_VehicleDeRez_Transition}
 ]
}
